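# Material function that adds an animated, optionally voice-reactive glowing visor to a character material's emissive, behind the UseAnimatedVisor switch and a mask channel.
# SubUV frames of a distance-field texture are thresholded into crisp shapes, tinted by an emissive texture, and combined by Max across two layers, the second parallax-offset.
# With UseEventVO, audio collection amplitude drives frame and brightness.
# A third layer and VisorState texture selection are left unconnected.
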MaterialFunction /Game/Characters/Player/Female/Medium/Bodies/F_MED_Binary/Materials/MF_Visor
graph {nodes [
  n1 = Power(Base=Multiply(A=CollectionParameter() {collection=/Game/Athena/Events/AudioAnalysis/MPC_Audio_Reactive_Parameters, parameterName=Amplitude_VO_B}) {constB=3}) {constExponent=1.112}
  Cosine(None=@n1)
  n2 = GetMaterialAttributes(None=FunctionInput() {inputName=Material, inputType=FunctionInput_MaterialAttributes}) {attributeGetTypes [#1]}
  n3 = FunctionInput(Preview=Constant() {r=1}) {bUsePreviewValueAsDefault=true, inputName=Mask, inputType=FunctionInput_Vector4}
  n4 = TextureObjectParameter() {
   group=Visor
   parameterName=Visor_EmissiveDistanceField
   samplerType=SAMPLERTYPE_LinearGrayscale
   texture=/Game/Global/Textures/Default/Blanks/T_BlackLinearGrayscale
  }
  n5 = TextureCoordinate()
  n6 = ScalarParameter() {defaultValue=2, group=Visor, parameterName=Visor_SubUV_Frames}
  n7 = Time()
  n8 = CollectionParameter() {collection=&MPC_Audio_Reactive_Parameters, parameterName=Amplitude_VO_B}
  n9 = Multiply(A=@n8) {constB=1.8}
  n10 = Ceil(None=@n8)
  n11 = Multiply(
   A=StaticSwitchParameter(True=LinearInterpolate(A=@n7, B=@n9, Alpha=@n10), False=@n7) {group=SmartTuna, parameterName=UseEventVO}
   B=ScalarParameter() {defaultValue=2, group=Visor, parameterName=Visor_SubUV_Speed}
  )
  n12 = MaterialFunctionCall(Float3=MaterialFunctionCall!(Texture=@n4, UVs=@n5, SubImages=@n6, Frame=@n11) {materialFunction=/Engine/Functions/Engine_MaterialFunctions03/Texturing/SubUV_Function}.RGB) {materialFunction=&BreakOutFloat3Components}
  n13 = ScalarParameter() {defaultValue=-0.48, group=Visor, parameterName=Visor_LineThick}
  n14 = Multiply(A=Saturate(None=Add(A=@n12.R, B=@n13) {constB=-0.48}), B=Constant() {r=16})
  n15 = NamedRerouteUsage()
  n16 = TextureSample(UVs=@n5, Tex=@n15)
  n17 = Multiply(
   A=Multiply(
    A=Saturate(None=@n14)
    B=StaticComponentMaskParameter(None=@n16.RGB) {defaultB=true, defaultG=true, defaultR=true, group=Visor, parameterName=Visor_Emissive Component Mask}
   )
   B=ScalarParameter() {defaultValue=5, group=Visor, parameterName=Visor_Layer1_Strength}
  )
  n18 = LinearInterpolate(Alpha=Saturate(None=Sine(None=@n1))) {constA=0.8, constB=6}
  n19 = LinearInterpolate(Alpha=Saturate(None=@n1)) {constA=0.75, constB=8}
  n20 = BumpOffset(
   Coordinate=TextureCoordinate()
   Height=ScalarParameter() {group=Visor, parameterName=VisorHeight}
   HeightRatioInput=ScalarParameter() {defaultValue=0.015, parameterName=VisorLayer2HeightRatio}
  ) {referencePlane=1}
  n21 = MaterialFunctionCall(Float3=MaterialFunctionCall(Texture=@n4, UVs=@n20, SubImages=@n6, Frame=@n11) {materialFunction=&SubUV_Function}.RGB) {materialFunction=&BreakOutFloat3Components}
  n22 = Multiply(A=Saturate(None=Add(A=@n21.R, B=@n13) {constB=-0.48}), B=Constant() {r=16})
  n23 = TextureSample(UVs=@n20, Tex=@n15)
  n24 = Multiply(
   A=Multiply(
    A=Saturate(None=@n22)
    B=StaticComponentMaskParameter(None=@n23.RGB) {defaultB=true, defaultG=true, defaultR=true, group=Visor, parameterName=Visor_Emissive Component Mask}
   )
   B=ScalarParameter() {defaultValue=1, group=Visor, parameterName=Visor_Layer2_Strength}
  )
  n25 = Max(
   A=StaticSwitchParameter(True=Multiply(A=@n17, B=@n18), False=@n17) {group=SmartTuna, parameterName=UseEventVO}
   B=StaticSwitchParameter(True=Multiply(A=@n24, B=@n19), False=@n24) {group=SmartTuna, parameterName=UseEventVO}
  )
  n26 = Add(
   A=@n2.EmissiveColor
   B=Multiply(A=ChannelMaskParameter(None=@n3) {group=Visor, maskChannel=Green, parameterName=VisorMaskChannel, defaultValue=<rgba 0 1 0 0>}.R, B=@n25)
  )
  FunctionOutput(None=SetMaterialAttributes!(
    MaterialAttributes=@n2.MaterialAttributes
    EmissiveColor=StaticSwitchParameter(True=@n26, False=@n2.EmissiveColor) {parameterName=UseAnimatedVisor}
   ) {attributeSetTypes [#1]})
  n27 = TextureObjectParameter() {group=Visor, parameterName=Visor_Emissive, texture=&T_BlacksRGB}
  n28 = TextureObject() {
   samplerType=SAMPLERTYPE_Alpha
   texture=/BRCosmetics/Characters/Player/Male/Medium/Bodies/M_MED_SynthContact/Textures/T_SynthContact_Visor_01_Flipbook
  }
  n29 = TextureObject() {
   samplerType=SAMPLERTYPE_Alpha
   texture=/BRCosmetics/Characters/Player/Male/Medium/Bodies/M_MED_SynthContact/Textures/T_SynthContact_Visor_06_Flipbook
  }
  If(A=Constant() {r=1}, B=ScalarParameter() {group=SmartTuna, parameterName=VisorState}, "A > B"=@n27, "A == B"=@n28, "A < B"=@n29)
  n30 = BumpOffset(
   Coordinate=TextureCoordinate()
   Height=ScalarParameter() {group=Visor, parameterName=VisorHeight}
   HeightRatioInput=ScalarParameter() {defaultValue=0.025, group=Visor, parameterName=VisorLayer3HeightRatio}
  ) {referencePlane=1}
  n31 = Add(A=MaterialFunctionCall(Float3=MaterialFunctionCall(Texture=@n4, UVs=@n30, SubImages=@n6, Frame=@n11) {materialFunction=&SubUV_Function}.RGB) {materialFunction=&BreakOutFloat3Components}.R) {constB=-0.48}
  n32 = Saturate(None=Multiply(A=Saturate(None=@n31), B=Constant() {r=16}))
  n33 = StaticComponentMaskParameter(None=TextureSample(UVs=@n30, Tex=@n15).RGB) {defaultB=true, defaultG=true, defaultR=true, group=Visor, parameterName=Visor_Emissive Component Mask}
  Max(A=@n25, B=Multiply(A=Multiply(A=@n32, B=@n33), B=ScalarParameter() {defaultValue=0.25, group=Visor, parameterName=Visor_Layer3_Strength}))
  NamedRerouteDeclaration(None=StaticSwitchParameter(True=@n27, False=@n27) {group=SmartTuna, parameterName=UseEventVO}) {name=EmissiveTex, nodeColor=<rgba 0.8047 0 1 1>}
  ScalarParameter() {defaultValue=0.02, group=Visor, parameterName=VisorLayerHeightRatio}
 ]}
