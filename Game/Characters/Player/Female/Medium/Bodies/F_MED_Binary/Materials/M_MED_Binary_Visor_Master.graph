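# Master material for the visor of the Binary female medium player character, built on material attributes for skeletal meshes.
# Shared character PreFX output passes through switchable clear coat, thin film and metal LUT layers, then the visor function, before PostFX finishes it.
# One SkinFX mask texture drives these layers and the visor, with selectable channels for clear coat and thin film.
# Shading model comes from the material expression.
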Material /Game/Characters/Player/Female/Medium/Bodies/F_MED_Binary/Materials/M_MED_Binary_Visor_Master
graph {
 properties {
  bUseMaterialAttributes=true
  bUsedWithMeshDeformer=true
  bUsedWithMorphTargets=true
  bUsedWithSkeletalMesh=true
  pixelDepthOffsetMode=PDOM_Legacy
  shadingModel=MSM_FromMaterialExpression
  usedShadingModels=Default Lit | Clear Coat | Subsurface Profile
 }
 nodes [
  n1 = MaterialFunctionCall() {materialFunction=&BaseFN_CharacterFunction_PreFX}
  n2 = TextureSample(Tex=TextureObjectParameter() {group=Skin FX, parameterName=SkinFX_Mask})
  n3 = StaticSwitchParameter(
   True=MaterialFunctionCall!(Alpha=StaticComponentMaskParameter(None=@n2.RGBA) {defaultR=true, group=Skin FX, parameterName=Clear Coat Channel}, In=@n1.Result) {materialFunction=/Game/Weapons/WeaponSkins/MaterialFunction/Clearcoat}.Result
   False=@n1.Result
  ) {group=Skin FX, parameterName=Use Clear Coat}
  n4 = StaticSwitchParameter(
   True=MaterialFunctionCall!(Alpha=StaticComponentMaskParameter(None=@n2.RGBA) {defaultR=true, group=Skin FX, parameterName=ThinFilm_Channel}, Material Attributes=@n3) {materialFunction=/Game/Weapons/WeaponSkins/MaterialFunction/ThinFilmFunction}.Result
   False=@n3
  ) {group=Skin FX, parameterName=Use Thin Film}
  n5 = MaterialFunctionCall(Mask=@n2.RGBA, Materials=@n4) {materialFunction=&MF_MetalsLUT}
  MaterialFunctionCall(
   AO=@n1.AO
   Distance Blend=@n1.AO
   Fresnel=@n1.AO
   Gmap=@n1.AO
   Pre Skinned Local Position=@n1.AO
   Material=MaterialFunctionCall!(Mask=@n2.RGBA, Material=StaticSwitchParameter(True=@n5.Result, False=@n4) {group=Skin FX, parameterName=UseMetalLUT}) {materialFunction=&/MF_Visor}.Result
  ) {materialFunction=&BaseFN_CharacterFunction_PostFX}
 ]
}
